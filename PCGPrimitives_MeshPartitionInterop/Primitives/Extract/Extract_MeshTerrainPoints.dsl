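; PCG primitive graph that extracts mesh terrain points, as part of the PCG Primitives interop with Mesh Partition.
; It tags the primary input as "Input:0", then chains three subgraphs: inputs management, core process and output preparation.
; Graph parameters expose the input fallback, input processing, core process and output preparation graphs as overrides of the template defaults.
; The extraction logic itself lives in the CoreProcess subgraph.
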(graph Extract_MeshTerrainPoints
  (params (InputsProcessingGraph :type (SoftObject /Script/PCG.PCGGraphInterface) :default /PCGPrimitives/Primitives/_Template_/Primitives_InputsProcessing.Primitives_InputsProcessing) (InputsFallbackGraph :type (SoftObject /Script/PCG.PCGGraphInterface) :default /PCGPrimitives/Primitives/_Template_/Primitives_InputsFallback.Primitives_InputsFallback) (CoreProcessGraph :type (SoftObject /Script/PCG.PCGGraphInterface) :default /PCGPrimitives_MeshPartitionInterop/Primitives/Extract/Subgraphs/Extract_MeshTerrainPoints_CoreProcess.Extract_MeshTerrainPoints_CoreProcess) (PrepareOutputsGraph :type (SoftObject /Script/PCG.PCGGraphInterface) :default /PCGPrimitives/Primitives/_Template_/Primitives_PrepareOutputs.Primitives_PrepareOutputs))
  (inputs (PrimaryInput :type Any))
  (bind AddTags_0 (Add%20Tags :TagsToAdd "Input:0"))
  ;[0 0 0 0.5 32] Inputs
  (bind Subgraph_0 (Subgraph :Graph /PCGPrimitives/Primitives/_Template_/Subgraphs/Primitives_InputsManagement.Primitives_InputsManagement :InputsFallbackGraph InputsFallbackGraph :InputsProcessingGraph InputsProcessingGraph))
  ;[0 0 0 0.5 32] Core Process
  (bind Subgraph_2 (Subgraph :Graph /PCGPrimitives_MeshPartitionInterop/Primitives/Extract/Subgraphs/Extract_MeshTerrainPoints_CoreProcess.Extract_MeshTerrainPoints_CoreProcess :SubgraphOverride-param CoreProcessGraph))
  ;[0 0 0 0.5 32] Prepare Output
  (bind Subgraph_3 (Subgraph :Graph /PCGPrimitives/Primitives/_Template_/Primitives_PrepareOutputs.Primitives_PrepareOutputs :SubgraphOverride-param PrepareOutputsGraph))
  (connect DefaultInputNode PrimaryInput AddTags_0 In)
  (connect AddTags_0 Out Subgraph_0 In)
  (connect Subgraph_0 Out Subgraph_2 In)
  (connect Subgraph_2 Out Subgraph_3 In)
  (output Out Subgraph_3))
